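# UEFN button text block, derived from WBP_UEFN_Block_Base, that shows a single CommonTextBlock with state-driven styling.
# Transition events play per-state animations forward or backward on font material Trans_ scalars, choosing selected or normal variants via GetIsSelected.
# Instant-transition and deselection events set those scalars directly on the cached FontDynamicMaterial.
# Setter functions expose text layout, font and material parameters; font changes re-cache the dynamic material, and empty text collapses the widget.
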WidgetBlueprint /FortUILibrary/UGC/Widgets/Buttons/Blocks/WBP_UEFN_Block_Text
animations [
 Anim_Hovered_Transition {
  binding_count=1
  length_seconds=1
  object=&self:Anim_Hovered_Transition
  sequence {
   bindings [TextBlock {
     id=#1
     possessed_class=CommonTextBlock
     tracks [{
       type=MovieSceneWidgetMaterialTrack
       display_name=Font.FontMaterial
       sections [{
         type=MovieSceneParameterSection
         end_frame=60
         start_frame=0
         channels [Trans_Hovered {type=MovieSceneScriptingFloatChannel, keys [[0,0],[60,1]]}]
        }]
      }]
    }]
   display_rate {numerator=60}
  }
 }
 Anim_Focused_Transition {
  binding_count=1
  length_seconds=1
  object=&self:Anim_Focused_Transition
  sequence {
   bindings [TextBlock {
     id=#1
     possessed_class=CommonTextBlock
     tracks [{
       type=MovieSceneWidgetMaterialTrack
       display_name=Font.FontMaterial
       sections [{
         type=MovieSceneParameterSection
         end_frame=60
         start_frame=0
         channels [Trans_Focused {type=MovieSceneScriptingFloatChannel, keys [[0,0],[60,1]]}]
        }]
      }]
    }]
   display_rate {numerator=60}
  }
 }
 Anim_Pressed_Transition {
  binding_count=1
  length_seconds=1
  object=&self:Anim_Pressed_Transition
  sequence {bindings [TextBlock {
     id=#2
     possessed_class=CommonTextBlock
     tracks [{
       type=MovieSceneWidgetMaterialTrack
       display_name=Font.FontMaterial
       sections [{
         type=MovieSceneParameterSection
         end_frame=20
         start_frame=0
         channels [Trans_Pressed {type=MovieSceneScriptingFloatChannel, keys [[0,0],[20,1]]}]
        }]
      }]
    }]}
 }
 Anim_Disabled_Transition {
  binding_count=1
  length_seconds=1
  object=&self:Anim_Disabled_Transition
  sequence {bindings [TextBlock {
     id=#2
     possessed_class=CommonTextBlock
     tracks [{
       type=MovieSceneWidgetMaterialTrack
       display_name=Font.FontMaterial
       sections [{
         type=MovieSceneParameterSection
         end_frame=20
         start_frame=0
         channels [Trans_Disabled {type=MovieSceneScriptingFloatChannel, keys [[0,0],[20,1]]}]
        }]
      }]
    }]}
 }
 Anim_Selected_Transition {
  binding_count=1
  length_seconds=1
  object=&self:Anim_Selected_Transition
  sequence {bindings [TextBlock {
     id=#3
     possessed_class=CommonTextBlock
     tracks [{
       type=MovieSceneWidgetMaterialTrack
       display_name=Font.FontMaterial
       sections [{
         type=MovieSceneParameterSection
         end_frame=20
         start_frame=0
         channels [Trans_SelectedIdle {type=MovieSceneScriptingFloatChannel, keys [[0,0],[19.9997,1]]}]
        }]
      }]
    }]}
 }
 Anim_SelectedHovered_Transition {
  binding_count=1
  length_seconds=1
  object=&self:Anim_SelectedHovered_Transition
  sequence {bindings [TextBlock {
     id=#4
     possessed_class=CommonTextBlock
     tracks [{
       type=MovieSceneWidgetMaterialTrack
       display_name=Font.FontMaterial
       sections [{
         type=MovieSceneParameterSection
         end_frame=20
         start_frame=0
         channels [Trans_SelectedHovered {type=MovieSceneScriptingFloatChannel, keys [[0,0],[19.9997,1]]}]
        }]
      }]
    }]}
 }
 Anim_SelectedFocused_Transition {
  binding_count=1
  length_seconds=1
  object=&self:Anim_SelectedFocused_Transition
  sequence {bindings [TextBlock {
     id=#5
     possessed_class=CommonTextBlock
     tracks [{
       type=MovieSceneWidgetMaterialTrack
       display_name=Font.FontMaterial
       sections [{
         type=MovieSceneParameterSection
         end_frame=20
         start_frame=0
         channels [Trans_SelectedFocused {type=MovieSceneScriptingFloatChannel, keys [[0,0],[19.9997,1]]}]
        }]
      }]
    }]}
 }
 Anim_SelectedPressed_Transition {
  binding_count=1
  length_seconds=1
  object=&self:Anim_SelectedPressed_Transition
  sequence {bindings [TextBlock {
     id=#6
     possessed_class=CommonTextBlock
     tracks [{
       type=MovieSceneWidgetMaterialTrack
       display_name=Font.FontMaterial
       sections [{
         type=MovieSceneParameterSection
         end_frame=20
         start_frame=0
         channels [Trans_SelectedPressed {type=MovieSceneScriptingFloatChannel, keys [[0,0],[19.9997,1]]}]
        }]
      }]
    }]}
 }
 Anim_SelectedDisabled_Transition {
  binding_count=1
  length_seconds=1
  object=&self:Anim_SelectedDisabled_Transition
  sequence {bindings [TextBlock {
     id=#7
     possessed_class=CommonTextBlock
     tracks [{
       type=MovieSceneWidgetMaterialTrack
       display_name=Font.FontMaterial
       sections [{
         type=MovieSceneParameterSection
         end_frame=20
         start_frame=0
         channels [Trans_SelectedDisabled {type=MovieSceneScriptingFloatChannel, keys [[0,0],[19.9997,1]]}]
        }]
      }]
    }]}
 }
]
graphs [
 EventGraph {nodes [
   Event OnInstantTransitionToSelected() {is=title K2Node_Event}
   > StopAllAnimations(self:User Widget Object Reference)
   n1 = Get FontDynamicMaterial() -> (FontDynamicMaterial:Material Instance Dynamic Object Reference)
   > SetScalarParameterValue(self=@n1, ParameterName=Trans_SelectedIdle, Value=1)
   > SetScalarParameterValue(self=@n1, ParameterName=Trans_SelectedHovered, Value=0)
   > SetScalarParameterValue(self=@n1, ParameterName=Trans_SelectedFocused, Value=0)
   > SetScalarParameterValue(self=@n1, ParameterName=Trans_SelectedPressed, Value=0)
   > SetScalarParameterValue(self=@n1, ParameterName=Trans_SelectedDisabled, Value=0)
   Event OnTransitionSelectedRequested() {is=title K2Node_Event}
   > PlayTransition!(
    execute
    self:UIKit Block Object Reference
    Animation:Widget Animation Object Reference=Get Anim_Selected_Transition() -> (Anim_Selected_Transition:Widget Animation Object Reference)
    bPlayForward:bool=true
   ) -> (then)
   Event OnTransitionUnhoveredRequested() {is=title K2Node_Event}
   > n2 = Branch(Condition=GetIsSelected!() -> (IsSelected:bool))
   > SetScalarParameterValue(self=Get FontDynamicMaterial(), ParameterName=Trans_SelectedIdle, Value=1)
   > PlayTransition(Animation=Get Anim_SelectedHovered_Transition() -> (Anim_SelectedHovered_Transition:Widget Animation Object Reference), bPlayForward=false)
   StopAnimation(execute=@n2.else, InAnimation=Get Anim_Selected_Transition())
   > PlayTransition(Animation=Get Anim_Hovered_Transition() -> (Anim_Hovered_Transition:Widget Animation Object Reference), bPlayForward=false)
   Event OnTransitionFocusedRequested() {is=title K2Node_Event}
   > n3 = Branch(Condition=GetIsSelected())
   > PlayTransition(Animation=Get Anim_SelectedFocused_Transition() -> (Anim_SelectedFocused_Transition:Widget Animation Object Reference), bPlayForward=true)
   PlayTransition(execute=@n3.else, Animation=Get Anim_Focused_Transition() -> (Anim_Focused_Transition:Widget Animation Object Reference), bPlayForward=true)
   Event OnTransitionUnfocusedRequested() {is=title K2Node_Event}
   > n4 = Branch(Condition=GetIsSelected())
   > SetScalarParameterValue(self=Get FontDynamicMaterial(), ParameterName=Trans_SelectedIdle, Value=1)
   > PlayTransition(Animation=Get Anim_SelectedFocused_Transition(), bPlayForward=false)
   StopAnimation(execute=@n4.else, InAnimation=Get Anim_Selected_Transition())
   > PlayTransition(Animation=Get Anim_Focused_Transition(), bPlayForward=false)
   Event OnTransitionPressedRequested() {is=title K2Node_Event}
   > n5 = Branch(Condition=GetIsSelected())
   > PlayTransition(Animation=Get Anim_SelectedPressed_Transition() -> (Anim_SelectedPressed_Transition:Widget Animation Object Reference), bPlayForward=true)
   PlayTransition(execute=@n5.else, Animation=Get Anim_Pressed_Transition() -> (Anim_Pressed_Transition:Widget Animation Object Reference), bPlayForward=true)
   Event OnTransitionReleasedRequested() {is=title K2Node_Event}
   > n6 = Branch(Condition=GetIsSelected())
   > PlayTransition(Animation=Get Anim_SelectedPressed_Transition(), bPlayForward=false)
   PlayTransition(execute=@n6.else, Animation=Get Anim_Pressed_Transition(), bPlayForward=false)
   Event OnTransitionDisabledRequested() {is=title K2Node_Event}
   > n7 = Branch(Condition=GetIsSelected())
   > PlayTransition(Animation=Get Anim_SelectedDisabled_Transition() -> (Anim_SelectedDisabled_Transition:Widget Animation Object Reference), bPlayForward=true)
   PlayTransition(
    execute=@n7.else
    Animation=Get Anim_Disabled_Transition() -> (Anim_Disabled_Transition:Widget Animation Object Reference)
    bPlayForward=true
   )
   Event OnTransitionEnabledRequested() {is=title K2Node_Event}
   > n8 = Branch(Condition=GetIsSelected())
   > PlayTransition(Animation=Get Anim_SelectedDisabled_Transition(), bPlayForward=false)
   PlayTransition(execute=@n8.else, Animation=Get Anim_Disabled_Transition(), bPlayForward=false)
   Event OnTransitionHoveredRequested() {is=title K2Node_Event}
   > n9 = Branch(Condition=GetIsSelected())
   > PlayTransition(Animation=Get Anim_SelectedHovered_Transition(), bPlayForward=true)
   PlayTransition(execute=@n9.else, Animation=Get Anim_Hovered_Transition(), bPlayForward=true)
   Event OnTransitionDeselectedFocusedRequested() {is=title K2Node_Event}
   > PlayTransition(Animation=Get Anim_SelectedFocused_Transition(), bPlayForward=false)
   > StopAnimation(InAnimation=Get Anim_Selected_Transition())
   n10 = Get FontDynamicMaterial()
   > SetScalarParameterValue(self=@n10, ParameterName=Trans_SelectedIdle, Value=0)
   > SetScalarParameterValue(self=@n10, ParameterName=Trans_SelectedHovered, Value=0)
   > SetScalarParameterValue(self=@n10, ParameterName=Trans_SelectedPressed, Value=0)
   > SetScalarParameterValue(self=@n10, ParameterName=Trans_SelectedDisabled, Value=0)
   > SetScalarParameterValue(self=@n10, ParameterName=Trans_Pressed, Value=0)
   Event OnTransitionDeselectedIdleRequested() {is=title K2Node_Event}
   > PlayTransition(Animation=Get Anim_Selected_Transition(), bPlayForward=false)
   n11 = Get FontDynamicMaterial()
   > SetScalarParameterValue(self=@n11, ParameterName=Trans_SelectedHovered, Value=0)
   > SetScalarParameterValue(self=@n11, ParameterName=Trans_SelectedFocused, Value=0)
   > SetScalarParameterValue(self=@n11, ParameterName=Trans_SelectedPressed, Value=0)
   > SetScalarParameterValue(self=@n11, ParameterName=Trans_SelectedDisabled, Value=0)
   > SetScalarParameterValue(self=@n11, ParameterName=Trans_Hovered, Value=0)
   > SetScalarParameterValue(self=@n11, ParameterName=Trans_Focused, Value=0)
   > SetScalarParameterValue(self=@n11, ParameterName=Trans_Pressed, Value=0)
   > SetScalarParameterValue(self=@n11, ParameterName=Trans_Disabled, Value=0)
   Event OnInstantTransitionToEnabled() {is=title K2Node_Event}
   > StopAllAnimations()
   > n12 = Branch(Condition=GetIsSelected())
   n13 = Get FontDynamicMaterial()
   SetScalarParameterValue(execute=[@n14.then, @n12.then], self=@n13, ParameterName=Trans_SelectedHovered, Value=0)
   > SetScalarParameterValue(self=@n13, ParameterName=Trans_SelectedFocused, Value=0)
   > SetScalarParameterValue(self=@n13, ParameterName=Trans_SelectedPressed, Value=0)
   > SetScalarParameterValue(self=@n13, ParameterName=Trans_SelectedDisabled, Value=0)
   SetScalarParameterValue(execute=@n12.else, self=@n13, ParameterName=Trans_Hovered, Value=0)
   > SetScalarParameterValue(self=@n13, ParameterName=Trans_Focused, Value=0)
   > SetScalarParameterValue(self=@n13, ParameterName=Trans_Pressed, Value=0)
   > SetScalarParameterValue(self=@n13, ParameterName=Trans_Disabled, Value=0)
   > n14 = SetScalarParameterValue(self=@n13, ParameterName=Trans_SelectedIdle, Value=0)
   Event OnInstantTransitionToDisabled() {is=title K2Node_Event}
   > StopAllAnimations()
   > n15 = Branch(Condition=GetIsSelected())
   n16 = Get FontDynamicMaterial()
   > SetScalarParameterValue(self=@n16, ParameterName=Trans_SelectedDisabled, Value=1)
   SetScalarParameterValue(execute=@n15.else, self=@n16, ParameterName=Trans_Disabled, Value=1)
  ]}
 SetFont {nodes [
   n1 = SetFont!() -> (then, Font:"Slate Font Info Structure (by ref)") {is=title K2Node_FunctionEntry}
   n2 = Get TextBlock() -> (TextBlock:Common Text Object Reference)
   > Set Font!(execute, Font:Slate Font Info Structure=@n1, self:Text Object Reference=@n2) -> (then, Output_Get:Slate Font Info Structure)
   > n3 = GetDynamicFontMaterial!(execute, self:Text Object Reference=@n2) -> (then, ReturnValue:Material Instance Dynamic Object Reference)
   > Set FontDynamicMaterial(FontDynamicMaterial:Material Instance Dynamic Object Reference=@n3) -> (Output_Get:Material Instance Dynamic Object Reference)
  ]}
 SetFontMaterial {nodes [
   n1 = SetFontMaterial!() -> (then, In Material:Material Instance Object Reference) {is=title K2Node_FunctionEntry}
   n2 = Get TextBlock()
   > SetFontMaterial!(execute, self:Text Object Reference=@n2, InMaterial:Material Interface Object Reference=@n1) -> (then) {is=title K2Node_CallFunction}
   > n3 = GetDynamicFontMaterial(self=@n2)
   > Set FontDynamicMaterial(FontDynamicMaterial=@n3)
  ]}
 SetJustification {nodes [
   n1 = SetJustification!() -> (then, Justification:ETextJustify Enum=Left) {is=title K2Node_FunctionEntry}
   > Set Justification!(execute, Justification:ETextJustify Enum=@n1, self:Text Layout Widget Object Reference=Get TextBlock()) -> (then, Output_Get:ETextJustify Enum=Left)
  ]}
 SetMaterialDoubleVectorParameter {nodes [
   n1 = SetMaterialDoubleVectorParameter!() -> (then, ParameterName:Name, Value:Vector 4 Structure) {is=title K2Node_FunctionEntry}
   > SetDoubleVectorParameterValue!(
    execute
    self:Material Instance Dynamic Object Reference=Get FontDynamicMaterial()
    ParameterName:Name=@n1
    Value:Vector 4 Structure=@n1.Value
   ) -> (then)
  ]}
 SetMaterialForceMipLevelIsResidentParameter {nodes [
   n1 = SetMaterialForceMipLevelIsResidentParameter!() -> (
    then
    OverrideForceMiplevelsToBeResident:bool=~
    bForceMiplevelsToBeResidentValue:bool=~
    ForceDuration:float=~
    CinematicTextureGroups:int
    bFastResponse:bool=~
   ) {is=title K2Node_FunctionEntry}
   > SetForceMipLevelsToBeResident!(
    execute
    self:Material Interface Object Reference=Get FontDynamicMaterial()
    OverrideForceMiplevelsToBeResident:bool=@n1
    bForceMiplevelsToBeResidentValue:bool=@n1.bForceMiplevelsToBeResidentValue
    ForceDuration:float=@n1.ForceDuration
    CinematicTextureGroups:int=@n1.CinematicTextureGroups
    bFastResponse:bool=@n1.bFastResponse
   ) -> (then)
  ]}
 SetMaterialRuntimeVirtualTextureParameter {nodes [
   n1 = SetMaterialRuntimeVirtualTextureParameter!() -> (then, ParameterName:Name, Value:Runtime Virtual Texture Object Reference) {is=title K2Node_FunctionEntry}
   > SetRuntimeVirtualTextureParameterValue!(
    execute
    self:Material Instance Dynamic Object Reference=Get FontDynamicMaterial()
    ParameterName:Name=@n1
    Value:Runtime Virtual Texture Object Reference=@n1.Value
   ) -> (then)
  ]}
 SetMaterialScalarParameter {nodes [
   n1 = SetMaterialScalarParameter!() -> (then, ParameterName:Name, Value:float=~) {is=title K2Node_FunctionEntry}
   > SetScalarParameterValue(self=Get FontDynamicMaterial(), ParameterName=@n1, Value=@n1.Value)
  ]}
 SetMaterialSparseVolumeTextureParameter {nodes [
   n1 = SetMaterialSparseVolumeTextureParameter!() -> (then, ParameterName:Name, Value:Sparse Volume Texture Object Reference) {is=title K2Node_FunctionEntry}
   > SetSparseVolumeTextureParameterValue!(
    execute
    self:Material Instance Dynamic Object Reference=Get FontDynamicMaterial()
    ParameterName:Name=@n1
    Value:Sparse Volume Texture Object Reference=@n1.Value
   ) -> (then)
  ]}
 SetMaterialTextureParameter {nodes [
   n1 = SetMaterialTextureParameter!() -> (then, ParameterName:Name, Value:Texture Object Reference) {is=title K2Node_FunctionEntry}
   > SetTextureParameterValue(self=Get FontDynamicMaterial(), ParameterName=@n1, Value=@n1.Value)
  ]}
 SetMaterialVectorParameter {nodes [
   n1 = SetMaterialVectorParameter!() -> (then, ParameterName:Name, Value:Linear Color Structure) {is=title K2Node_FunctionEntry}
   > SetVectorParameterValue(self=Get FontDynamicMaterial(), ParameterName=@n1, Value=@n1.Value)
  ]}
 SetMobileFontSizeMultiplier {nodes [
   n1 = SetMobileFontSizeMultiplier!() -> (then, MobileFontSizeMultiplier:double=~) {is=title K2Node_FunctionEntry}
   > SetMobileFontSizeMultiplier!(execute, self:Common Text Object Reference=Get TextBlock(), InMobileFontSizeMultiplier:float=@n1) -> (then) {is=title K2Node_CallFunction}
  ]}
 SetText {nodes [
   n1 = SetText!() -> (then, Text:Text) {is=title K2Node_FunctionEntry}
   > n2 = Set Text(Text=@n1, self=Get TextBlock())
   n3 = TextIsEmpty(InText=@n2)
   > n4 = Branch(Condition=@n3)
   > n5 = SetVisibility(InVisibility=Collapsed)
   Return Node!(execute=[@n6.then, @n5.then], Empty Text:bool=@n3)
   n6 = SetVisibility(execute=@n4.else, InVisibility=SelfHitTestInvisible)
  ]}
 SetTransformPolicy {nodes [
   n1 = SetTransformPolicy!() -> (then, Transform Policy:ETextTransformPolicy Enum=None) {is=title K2Node_FunctionEntry}
   > Set TextTransformPolicy!(execute, TextTransformPolicy:ETextTransformPolicy Enum=@n1, self:Text Object Reference=Get TextBlock()) -> (then, Output_Get:ETextTransformPolicy Enum=None)
  ]}
 SetWrapping {nodes [
   n1 = SetWrapping!() -> (then, IsAutoWrap:bool=~, WrapTextAt:int) {is=title K2Node_FunctionEntry}
   n2 = Get TextBlock()
   > SetAutoWrapText!(execute, self:Text Object Reference=@n2, InAutoTextWrap:bool=@n1) -> (then)
   > Branch(Condition=@n1)
   > SetWrapTextWidth!(execute, self:Common Text Object Reference=@n2, InWrapTextAt:int=@n1.WrapTextAt) -> (then)
  ]}
]
variables [
 Font {
  type=SlateFontInfo
  default {
   bForceMonospaced=false
   bMaterialIsStencil=false
   fontMaterial=/FortUILibrary/UGC/Materials/M_UEFN_UI_Color_WithSelect
   fontObject=&BurbankSmall
   letterSpacing=0
   monospacedWidth=1
   size=15
   skewAmount=0
   typefaceFontName=Black
   outlineSettings {
    bApplyOutlineToDropShadows=false
    bMiteredCorners=false
    bSeparateFillAlpha=false
    outlineMaterial=null
    outlineSize=0
    outlineColor=<rgba 0 0 0 1>
   }
  }
 }
 FontDynamicMaterial {type=MaterialInstanceDynamic}
]
widgets {
 info {
  inheritedWidgetCount=0
  parentClass=&/WBP_UEFN_Block_Base.WBP_UEFN_Block_Base_C
  rootWidgetClass=&CommonTextBlock
  widgetCount=1
 }
 widgets [TextBlock {
   widgetClassPath=&CommonTextBlock
   bIsVariable=true
   widget=TextBlock
   properties {
    bAutoCollapseWithEmptyText=true
    justification=Center
    mobileFontSizeMultiplier=1.25
    scrollStyle=&TextScrollStyle-Base_C
    text=TextBlock
    textTransformPolicy=ToUpper
    font {
     bForceMonospaced=false
     bMaterialIsStencil=false
     fontMaterial=&M_UEFN_UI_Color_WithSelect
     fontObject=&BurbankSmall
     letterSpacing=0
     monospacedWidth=1
     size=15
     skewAmount=0
     typefaceFontName=Black
     outlineSettings {
      bApplyOutlineToDropShadows=false
      bMiteredCorners=false
      bSeparateFillAlpha=false
      outlineMaterial=null
      outlineSize=0
      outlineColor=<rgba 0 0 0 1>
     }
    }
    shadowOffset=<0 0>
   }
  }]
}
