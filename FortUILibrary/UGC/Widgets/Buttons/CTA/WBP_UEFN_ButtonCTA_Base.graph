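# Base UEFN call-to-action button widget on FortCTAButton, with primary and secondary text and an input prompt.
# Registered text and background UIKit blocks transition with button state, alongside a press scale animation and optional sounds.
# UpdatePrompt sets icon and hold-ring visibility by input type, while hold progress drives a material parameter and dispatchers.
# PreConstruct applies fonts, materials and mobile-aware sizing; non-mobile height can grow in grid-unit steps.
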WidgetBlueprint /FortUILibrary/UGC/Widgets/Buttons/CTA/WBP_UEFN_ButtonCTA_Base
animations [Anim_Pressed_Base {
  binding_count=2
  length_seconds=0.1
  object=&self:Anim_Pressed_Base
  sequence {bindings [
    VerticalBox_Text {
     id=#1
     possessed_class=VerticalBox
     tracks [{
       type=MovieScene2DTransformTrack
       sections [{
         type=MovieScene2DTransformSection
         channels [
          Scale.X {type=MovieSceneScriptingFloatChannel, keys [[0,1],[2,0.95]]}
          Scale.Y {type=MovieSceneScriptingFloatChannel, keys [[0,1],[2,0.975]]}
         ]
        }]
      }]
    }
    Root {
     id=#2
     possessed_class=GridPanel
     tracks [{
       type=MovieScene2DTransformTrack
       sections [{
         type=MovieScene2DTransformSection
         end_frame=2
         start_frame=0
         channels [
          Scale.X {type=MovieSceneScriptingFloatChannel, keys [[0,1],[2,0.975]]}
          Scale.Y {type=MovieSceneScriptingFloatChannel, keys [[0,1],[2,0.975]]}
         ]
        }]
      }]
    }
   ]}
 }]
graphs [
 Cleanup {nodes [
   n1 = Cleanup!() -> (then, Widget:Widget Object Reference) {is=title K2Node_FunctionEntry}
   > Branch(Condition=IsValid(Object=@n1))
   > RemoveChild(self=GetParent!(self:Widget Object Reference=@n1) -> (ReturnValue:Panel Widget Object Reference), Content=@n1)
  ]}
 EventGraph {nodes [
   n1 = "On Input Icon Updated (InputActionWidget)"() {is=title K2Node_ComponentBoundEvent}
   Set Icon Dark Mode!(execute=[@n7.then, @n1.then], self, IsDark:bool=Get IsDarkMode() -> (IsDarkMode:bool)) -> (then) {is=title K2Node_CallFunction}
   n2 = Create Event()
   n3 = Get EnhancedInputLocalPlayerSubsystem() -> (ReturnValue:Enhanced Input Local Player Subsystem Object Reference) {is=title K2Node_GetSubsystem}
   > Bind Event to OnControlMappingsRebuilt(self:Enhanced Input Local Player Subsystem Object Reference=@n3, Delegate=@n2) {is=title K2Node_AddDelegate}
   OnControlMappingsRebuilt() {is=title K2Node_CustomEvent}
   > UpdatePromptVisibility!(execute, self, InputType:ECommonInputType Enum=Get CurrentInputType() -> (CurrentInputType:ECommonInputType Enum)) -> (then)
   Event PreConstruct()
   > n4 = Sequence()
   > SetText!(execute, self:Fort CTAButton Object Reference, ButtonText:"Text (by ref)"=Get Text!() -> (Text)) -> (then)
   > SetSecondaryText!(execute, self:Fort CTAButton Object Reference, ButtonSecondaryText:"Text (by ref)"=Get SecondaryText() -> (SecondaryText:Text)) -> (then)
   > InitializeButton()
   RegisterBlocks(execute=@n4.then_1)
   > SetBlocksTiming!(
    execute
    self:UIKit Modular Button Object Reference
    Timing_Hovering:float=0
    Timing_Unhovering:float=0
    Timing_Focusing:float=0.3
    Timing_Unfocusing:float=0.15
    Timing_Pressing:float=0.1
    Timing_Releasing:float=0.1
    Timing_Disabling:float=0.1
    Timing_Enabling:float=0.1
    Timing_Selecting:float=0
    Timing_DeselectingIdle:float=0
    Timing_DeselectingFocused:float=0
   ) -> (then)
   > SetBlockTiming!(
    execute
    self:UIKit Block Object Reference=Get Block_Text() -> (Block_Text:WBP UEFN Block Text Object Reference)
    Timing_Hovering:float=0
    Timing_Unhovering:float=0
    Timing_Focusing:float=0.15
    Timing_Unfocusing:float=0.15
    Timing_Pressing:float=0.1
    Timing_Releasing:float=0.1
    Timing_Disabling:float=0.1
    Timing_Enabling:float=0.1
    Timing_Selecting:float=0
    Timing_DeselectingIdle:float=0
    Timing_DeselectingFocused:float=0
   ) -> (then)
   > SetBlockTiming(
    self=Get Block_SecondaryText() -> (Block_SecondaryText:WBP UEFN Block Text Object Reference)
    Timing_Hovering=0.15
    Timing_Unhovering=0.1
    Timing_Focusing=0.2
    Timing_Unfocusing=0.15
    Timing_Pressing=0.1
    Timing_Releasing=0.1
    Timing_Disabling=0.1
    Timing_Enabling=0.1
    Timing_Selecting=0
    Timing_DeselectingIdle=0
    Timing_DeselectingFocused=0
   )
   > ResetButtonState()
   Event On Pressed() {is=title K2Node_Event}
   > RequestTransitionOnBlocks!(execute, self:UIKit Modular Button Object Reference, State:EUIKitBlockVisualState Enum=Pressed) -> (then)
   > PlayAnimationForward(InAnimation=Get Anim_Pressed_Base() -> (Anim_Pressed_Base:Widget Animation Object Reference))
   > n5 = Branch(Condition=IsPromptHoldable!() -> (Holdable:bool))
   > PlaySoundHoldPressed()
   PlaySoundPressed(execute=@n5.else)
   n6 = Event OnActionProgress!() -> (OutputDelegate, then, HeldPercent:float) {is=title K2Node_Event}
   > SetHoldPercentage!(execute, self, HeldPercentageValue:double=@n6.HeldPercent) -> (then)
   > Call On Hold Action Progress!(execute:Exec, self:WBP UEFN Button CTA Base Object Reference, HeldPercent:double=@n6.HeldPercent) -> (then:Exec) {is=title K2Node_CallDelegate}
   Event Construct()
   > n7 = SetPrompt()
   Event Destruct()
   > Fade Out Sound()
   > Unbind Event from OnControlMappingsRebuilt(self:Enhanced Input Local Player Subsystem Object Reference=@n3, Delegate=@n2) {is=title K2Node_RemoveDelegate}
   Event On Input Action Triggered() {is=title K2Node_Event}
   > PlaySoundPressed()
   Event On Focused() {is=title K2Node_Event}
   > Call On Focus Changed!(execute:Exec, self:WBP UEFN Button CTA Base Object Reference, HasFocus:bool=true) -> (then:Exec) {is=title K2Node_CallDelegate}
   Event On Unfocused() {is=title K2Node_Event}
   > Call On Focus Changed()
   Event BP_OnInputActionUpdated() {is=title K2Node_Event}
   > UpdatePrompt!(execute, self, InputType:ECommonInputType Enum=Get CurrentInputType()) -> (then)
   Event On Released() {is=title K2Node_Event}
   > RequestTransitionOnBlocks(State=Released)
   > PlayAnimationReverse(InAnimation=Get Anim_Pressed_Base(), PlaybackSpeed=2)
   n8 = Get CurrentHoldProgress() -> (CurrentHoldProgress:double)
   > Branch(Condition=AND Boolean!(A=IsPromptHoldable(), B="float > float"(A=@n8), C:bool="float < float"(A=@n8, B=1)) -> (ReturnValue))
   > Stop Sound Hold Released()
   Event On Clicked()
   > Branch(Condition=IsPromptHoldable())
   > SetHoldPercentage(HeldPercentageValue=0)
   > ResetButtonState()
   > PlaySoundHoldCompleted()
   > Call On Hold Completed!(execute:Exec, self:WBP UEFN Button CTA Base Object Reference) -> (then:Exec) {is=title K2Node_CallDelegate}
   Event On Disabled() {is=title K2Node_Event}
   > n9 = Set LastEnabledEvent(LastEnabledEvent:bool=false) -> (Output_Get:bool)
   DelayUntilNextTick!(execute=[@n9.then, @n10.then]) -> (then)
   > SetButtonIsEnabled()
   Event On Hovered()
   > ShowPromptFallbackAction!(execute, self, Show:bool=true) -> (then)
   > RequestTransitionOnBlocks(State=Focused)
   > PlaySoundHovered()
   Event On Unhovered()
   > ShowPromptFallbackAction(Show=false)
   > RequestTransitionOnBlocks(State=Unfocused)
   Event On Enabled() {is=title K2Node_Event}
   > n10 = Set LastEnabledEvent(LastEnabledEvent=true)
   n11 = Event BP_SetText!() -> (OutputDelegate, then, ButtonText:"Text (by ref)") {is=title K2Node_Event}
   > SetText!(execute, self:WBP UEFN Block Text Object Reference=Get Block_Text(), Text:Text=@n11.ButtonText) -> (then, Empty Text:bool)
   n12 = Event BP_SetSecondaryText!() -> (OutputDelegate, then, ButtonSecondaryText:"Text (by ref)") {is=title K2Node_Event}
   > n13 = SetText(self=Get Block_SecondaryText(), Text=@n12.ButtonSecondaryText)
   > Set SingleText(SingleText:bool=@n13) -> (Output_Get:bool)
   > UpdateButtonSIze()
   n14 = Event On Input Method Changed!() -> (OutputDelegate, then, CurrentInputType:ECommonInputType Enum) {is=title K2Node_Event}
   > n15 = Set CurrentInputType(CurrentInputType:ECommonInputType Enum=@n14.CurrentInputType) -> (Output_Get:ECommonInputType Enum)
   > UpdatePrompt(InputType=@n15)
   > ResetButtonState()
   > Fade Out Sound()
  ]}
 Fade Out Sound {nodes [
   Fade Out Sound() {is=title K2Node_FunctionEntry}
   > n1 = Sequence()
   > n2 = Get!(execute:Exec) -> (then:Exec, else:Exec, PressedSoundAudioComponent:Audio Component Object Reference)
   > FadeOut!(
    execute
    self:Audio Component Object Reference=@n2
    FadeOutDuration:float=Get FadeOutTime() -> (FadeOutTime:double)
    FadeVolumeLevel:float=0
    FadeCurve:EAudioFaderCurve Enum=Linear
   ) -> (then)
   n3 = Get!(execute=@n1.then_1) -> (then, else, HoldSoundAudioComponent:Audio Component Object Reference)
   > FadeOut(self=@n3, FadeOutDuration=Get FadeOutTime(), FadeVolumeLevel=0, FadeCurve=Linear)
  ]}
 InitializeButton {nodes [
   InitializeButton() {is=title K2Node_FunctionEntry}
   > UpdateFonts()
   > UpdateButtonSIze()
   > UpdateBackground()
  ]}
 IsPromptHoldable {nodes [
   IsPromptHoldable() {is=title K2Node_FunctionEntry}
   n1 = Get InputActionWidget() -> (InputActionWidget:Common Action Widget Object Reference)
   > Return Node!(
    execute
    Holdable:bool=OR Boolean(A=Get bRequiresHold() -> (bRequiresHold:bool), B=IsHeldAction!(self:Common Action Widget Object Reference=@n1) -> (ReturnValue:bool))
   )
  ]}
 IsPromptIcon {nodes [
   IsPromptIcon() {is=title K2Node_FunctionEntry}
   n1 = GetIcon!(self:Common Action Widget Object Reference=Get InputActionWidget()) -> (ReturnValue:Slate Brush Structure)
   > Return Node!(
    execute
    Valid:bool=IsValid(Object=Break Slate Brush!(SlateBrush:"Slate Brush Structure (by ref)"=@n1) -> (
      TintColor:Slate Color Structure
      DrawAs:ESlateBrushDrawType Enum
      Tiling:ESlateBrushTileType Enum
      Mirroring:ESlateBrushMirrorType Enum
      ImageSize:Vector 2D Structure
      Margin:Margin Structure
      ResourceObject:Object Reference
      OutlineSettings:Slate Brush Outline Settings Structure
     ) {is=title K2Node_BreakStruct}.ResourceObject)
   )
  ]}
 OnFocusChanged {nodes [OnFocusChanged!() -> (then, HasFocus:bool=~) {is=title K2Node_FunctionEntry}]}
 OnHoldActionProgress {nodes [OnHoldActionProgress!() -> (then, HeldPercent:double=~) {is=title K2Node_FunctionEntry}]}
 OnHoldCompleted {nodes [OnHoldCompleted() {is=title K2Node_FunctionEntry}]}
 PlaySoundHoldCompleted {nodes [
   PlaySoundHoldCompleted() {is=title K2Node_FunctionEntry}
   > n1 = Sequence()
   > n2 = Get!(execute) -> (then, else, HoldCompletedSound:Sound Base Object Reference)
   > SpawnSound2D!(
    execute
    Sound:Sound Base Object Reference=@n2
    VolumeMultiplier:float=1
    PitchMultiplier:float=1
    StartTime:float=0
    ConcurrencySettings:Sound Concurrency Object Reference
    bPersistAcrossLevelTransition:bool=true
    bAutoDestroy:bool=true
   ) -> (then, ReturnValue:Audio Component Object Reference)
   Fade Out Sound(execute=@n1.then_1) {is=title K2Node_CallFunction}
  ]}
 PlaySoundHoldPressed {nodes [
   PlaySoundHoldPressed() {is=title K2Node_FunctionEntry}
   > n1 = Sequence()
   > n2 = Get!(execute) -> (then, else, HoldPressedSound:Sound Base Object Reference)
   > n3 = SpawnSound2D(Sound=@n2, VolumeMultiplier=1, PitchMultiplier=1, StartTime=0, bPersistAcrossLevelTransition=false, bAutoDestroy=true)
   > Set PressedSoundAudioComponent(PressedSoundAudioComponent:Audio Component Object Reference=@n3) -> (Output_Get:Audio Component Object Reference)
   n4 = Get!(execute=@n1.then_1) -> (then, else, HoldSound:Sound Base Object Reference)
   > n5 = SpawnSound2D(Sound=@n4, VolumeMultiplier=1, PitchMultiplier=1, StartTime=0, bPersistAcrossLevelTransition=false, bAutoDestroy=true)
   > Set HoldSoundAudioComponent(HoldSoundAudioComponent:Audio Component Object Reference=@n5) -> (Output_Get:Audio Component Object Reference)
  ]}
 PlaySoundHovered {nodes [
   PlaySoundHovered() {is=title K2Node_FunctionEntry}
   > n1 = Get!(execute) -> (then, else, HoveredSound:Sound Base Object Reference)
   > SpawnSound2D(Sound=@n1, VolumeMultiplier=1, PitchMultiplier=1, StartTime=0, bPersistAcrossLevelTransition=false, bAutoDestroy=true)
  ]}
 PlaySoundPressed {nodes [
   PlaySoundPressed() {is=title K2Node_FunctionEntry}
   > n1 = Get!(execute) -> (then, else, PressedSound:Sound Base Object Reference)
   > SpawnSound2D(Sound=@n1, VolumeMultiplier=1, PitchMultiplier=1, StartTime=0, bPersistAcrossLevelTransition=false, bAutoDestroy=true)
  ]}
 RegisterBlocks {nodes [
   RegisterBlocks() {is=title K2Node_FunctionEntry}
   > RegisterModularBlock!(execute, self:UIKit Modular Button Object Reference, Block:UIKit Block Object Reference=Get Block_Text()) -> (then)
   > RegisterModularBlock(Block=Get Block_SecondaryText())
   > RegisterModularBlock(Block=Get Block_Background() -> (Block_Background:WBP UEFN Block Background Object Reference))
  ]}
 ResetButtonState {nodes [
   ResetButtonState() {is=title K2Node_FunctionEntry}
   > StopAllAnimations(self:User Widget Object Reference)
   > RequestInstantTransitionOnBlocks!(
    execute
    self:UIKit Modular Button Object Reference
    State:EUIKitBlockInstantTransitionState Enum=Select(
     Option 0:EUIKitBlockInstantTransitionState Enum=InstantDisabled
     Option 1:EUIKitBlockInstantTransitionState Enum=InstantEnabled
     Index=GetIsEnabled!(self:Widget Object Reference) -> (ReturnValue:bool)
    ) -> (ReturnValue:EUIKitBlockInstantTransitionState Enum)
   ) -> (then)
  ]}
 Set Icon Dark Mode {nodes [
   Set Icon Dark Mode!() -> (then, IsDark:bool=~) {is=title K2Node_FunctionEntry}
   > n1 = GetIconDynamicMaterial!(execute, self:Common Action Widget Object Reference=Get InputActionWidget()) -> (then, ReturnValue:Material Instance Dynamic Object Reference)
   > SetScalarParameterValue(self=@n1, ParameterName=Use Dark, Value="To Float (Boolean)"(InBool=Get IsDark() -> (IsDark:bool)))
   > SetScalarParameterValue(self=@n1, ParameterName=Outline Opacity, Value=Select(Option 0:float=0.5, Option 1:float=0, Index=Get IsDark()) -> (ReturnValue:float))
  ]}
 SetBackgroundMaterial {nodes [
   n1 = SetBackgroundMaterial!() -> (then, Background:Material Instance Object Reference) {is=title K2Node_FunctionEntry}
   > n2 = Set BackgroundMaterial(BackgroundMaterial:Material Instance Object Reference=@n1) -> (Output_Get:Material Instance Object Reference)
   > SetBackgroundMaterial!(
    execute
    self:WBP UEFN Block Background Base Object Reference=Get Block_Background()
    Material Instance:"Material Instance Object Reference (by ref)"=@n2
   ) -> (then) {is=title K2Node_CallFunction}
  ]}
 SetBorderMaterial {nodes [
   n1 = SetBorderMaterial!() -> (then, BorderMaterial:Material Instance Object Reference) {is=title K2Node_FunctionEntry}
   > n2 = Set BorderMaterial(BorderMaterial:Material Instance Object Reference=@n1) -> (Output_Get:Material Instance Object Reference)
   > SetOutlineMaterial!(
    execute
    self:WBP UEFN Block Background Base Object Reference=Get Block_Background()
    Material Instance:"Material Instance Object Reference (by ref)"=@n2
   ) -> (then)
  ]}
 SetButtonIsEnabled {nodes [
   SetButtonIsEnabled() {is=title K2Node_FunctionEntry}
   > n1 = Branch(Condition=Get LastEnabledEvent() -> (LastEnabledEvent:bool))
   > n2 = Branch(Condition=IsHovered())
   > n3 = RequestTransitionOnBlocks(State=Enabled)
   SetPrompt(execute=[@n3.then, @n4.then])
   n4 = RequestInstantTransitionOnBlocks(execute=@n2.else, State=InstantEnabled)
   StopAllAnimations(execute=@n1.else)
   > RequestTransitionOnBlocks(State=Disabled)
   > SetVisibility(self=Get Prompt() -> (Prompt:Grid Panel Object Reference), InVisibility=Collapsed)
  ]}
 SetHoldPercentage {nodes [
   n1 = SetHoldPercentage!() -> (then, HeldPercentageValue:double=~) {is=title K2Node_FunctionEntry}
   > n2 = Set CurrentHoldProgress(CurrentHoldProgress:double=@n1) -> (Output_Get:double)
   > SetScalarParameterValue(
    self=Get HoldProgressDynamicMaterial() -> (HoldProgressDynamicMaterial:Material Instance Dynamic Object Reference)
    ParameterName=Percentage
    Value=@n2
   )
  ]}
 SetPrompt {nodes [
   SetPrompt() {is=title K2Node_FunctionEntry}
   n1 = Get CommonInputSubsystem!() -> (ReturnValue:Common Input Object Reference) {is=title K2Node_GetBlueprintContextEx}
   > Branch(Condition=IsValid(Object=@n1))
   > n2 = GetDynamicMaterial(self=Get HoldProgressBar() -> (HoldProgressBar:Image Object Reference))
   > Set HoldProgressDynamicMaterial(HoldProgressDynamicMaterial:Material Instance Dynamic Object Reference=@n2) -> (Output_Get:Material Instance Dynamic Object Reference)
   > n3 = Set CurrentInputType(CurrentInputType=GetCurrentInputType(self=@n1))
   > UpdatePrompt(InputType=@n3)
  ]}
 ShowInputActionText {nodes [
   ShowInputActionText() {is=title K2Node_FunctionEntry}
   > SetText(self=Get Block_Text(), Text=GetDisplayText!(self:Common Action Widget Object Reference=Get InputActionWidget()) -> (ReturnValue:Text))
  ]}
 ShowPromptFallbackAction {nodes [
   n1 = ShowPromptFallbackAction!() -> (then, Show:bool=~) {is=title K2Node_FunctionEntry}
   n2 = Get CurrentInputType()
   > Branch(Condition=AND Boolean!(A="Equal (Enum)"(A=@n2, B=Gamepad), B=Get bShouldUseFallbackDefaultInputAction() -> (bShouldUseFallbackDefaultInputAction:bool)) -> (ReturnValue))
   > SetVisibility(
    self=Get Prompt()
    InVisibility=Select(Option 0:ESlateVisibility Enum=Collapsed, Option 1:ESlateVisibility Enum=SelfHitTestInvisible, Index=@n1) -> (ReturnValue:ESlateVisibility Enum)
   )
  ]}
 Stop Sound Hold Released {nodes [
   Stop Sound Hold Released() {is=title K2Node_FunctionEntry}
   > n1 = Sequence()
   > n2 = Get!(execute) -> (then, else, HoldReleasedSound:Sound Base Object Reference)
   > SpawnSound2D(Sound=@n2, VolumeMultiplier=1, PitchMultiplier=1, StartTime=0, bPersistAcrossLevelTransition=false, bAutoDestroy=true)
   Fade Out Sound(execute=@n1.then_1)
  ]}
 UnregisterBlocks {nodes [
   UnregisterBlocks() {is=title K2Node_FunctionEntry}
   > UnregisterAllModularBlocks(self:UIKit Modular Button Object Reference)
  ]}
 UpdateBackground {nodes [
   UpdateBackground() {is=title K2Node_FunctionEntry}
   n1 = Get Block_Background()
   > SetBackgroundMaterial(self=@n1, Material Instance=Get BackgroundMaterial() -> (BackgroundMaterial:Material Instance Object Reference))
   > SetOutlineMaterial(self=@n1, Material Instance=Get BorderMaterial() -> (BorderMaterial:Material Instance Object Reference))
   > SetCornerRadiuses!(
    execute
    self:WBP UEFN Block Background Object Reference=@n1
    CornerRadius:float=Get Corner Radius() -> (Corner Radius:float)
    RadiusInPixels:bool=Get Radius in Pixels() -> (Radius in Pixels:bool)
   ) -> (then)
  ]}
 UpdateButtonSIze {nodes [
   UpdateButtonSIze() {is=title K2Node_FunctionEntry}
   > n1 = Branch(Condition=IsMobileGame())
   > SetMinDimensions!(
    execute
    self:Common Button Base Object Reference
    InMinWidth:int=Get MobileMinWidth() -> (MobileMinWidth:int=0)
    InMinHeight:int=Select(
     Option 0:int=Get MobileButtonHeight() -> (MobileButtonHeight:int=0)
     Option 1:int=Get MobileMinHeight() -> (MobileMinHeight:int=0)
     Index=Get OverrideDefaultHeight() -> (OverrideDefaultHeight:bool)
    ) -> (ReturnValue:int)
   ) -> (then)
   > Set Padding!(
    execute
    Padding:Margin Structure=Select(
     Option 0:Margin Structure=Get MobileDoubleTextPadding() -> (MobileDoubleTextPadding:Margin Structure)
     Option 1:Margin Structure=Get MobileSingleTextPadding() -> (MobileSingleTextPadding:Margin Structure)
     Index=Get SingleText() -> (SingleText:bool)
    ) -> (ReturnValue:Margin Structure)
    self:Grid Slot Object Reference=SlotAsGridSlot(Widget=Get VerticalBox_Text() -> (VerticalBox_Text:Vertical Box Object Reference))
   ) -> (then, Output_Get:Margin Structure)
   > n2 = Set Padding(
    Padding=Get MobileDoubleTextSecondaryPadding() -> (MobileDoubleTextSecondaryPadding:Margin Structure)
    self:Vertical Box Slot Object Reference=SlotAsVerticalBoxSlot(Widget=Get Block_SecondaryText())
   )
   UpdateButtonSizeToGrid(execute=[@n3.then, @n2.then])
   SetMinDimensions(execute=@n1.else, InMinWidth=Get MinWidth() -> (MinWidth:int=0), InMinHeight=Get ButtonHeight() -> (ButtonHeight:int=0))
   > Set Padding(
    Padding=Select(
     Option 0=Get DoubleTextPadding() -> (DoubleTextPadding:Margin Structure)
     Option 1=Get SingleTextPadding() -> (SingleTextPadding:Margin Structure)
     Index=Get SingleText()
    )
    self:Grid Slot Object Reference=SlotAsGridSlot(Widget=Get VerticalBox_Text())
   )
   > n3 = Set Padding(
    Padding=Get DoubleTextSecondaryPadding() -> (DoubleTextSecondaryPadding:Margin Structure)
    self:Vertical Box Slot Object Reference=SlotAsVerticalBoxSlot(Widget=Get Block_SecondaryText())
   )
  ]}
 UpdateButtonSizeToGrid {nodes [
   UpdateButtonSizeToGrid() {is=title K2Node_FunctionEntry}
   > ForceLayoutPrepass(self:Widget Object Reference)
   n1 = Get GridUnit() -> (GridUnit:int=0)
   n2 = SelectInt(A=@n1, B=1, bPickA="Not Equal (Integer)"(A=@n1))
   n3 = Get Root() -> (Root:Grid Panel Object Reference)
   n4 = float / float(A=BreakVector2D(InVec=GetDesiredSize!(self:Widget Object Reference=@n3) -> (ReturnValue:Vector 2D Structure)).Y, B:int=@n2)
   > n5 = Set NewButtonHeight(NewButtonHeight:int=int * int(A=Ceil(A=@n4) {is=title K2Node_CallFunction}, B=@n2)) -> (Output_Get:int=0)
   > Branch(Condition="integer > integer"(A=@n5, B=Get ButtonHeight() -> (ButtonHeight=0)))
   > SetMinDimensions(InMinWidth=Get MinWidth() -> (MinWidth=0), InMinHeight=Get NewButtonHeight() -> (NewButtonHeight:int=0))
   > Return Node!(execute)
  ]}
 UpdateFonts {nodes [
   UpdateFonts() {is=title K2Node_FunctionEntry}
   > n1 = Sequence()
   n2 = Get Block_Text()
   > SetFont!(
    execute
    self:WBP UEFN Block Text Object Reference=@n2
    Font:"Slate Font Info Structure (by ref)"=Get PrimaryFont() -> (PrimaryFont:Slate Font Info Structure)
   ) -> (then)
   > SetMobileFontSizeMultiplier!(
    execute
    self:WBP UEFN Block Text Object Reference=@n2
    MobileFontSizeMultiplier:double=Select(
     Option 0:double=Get MobileFontSizeMultiplier() -> (MobileFontSizeMultiplier:double)
     Option 1:double=Get MobileFontSizeMultiplierOverride() -> (MobileFontSizeMultiplierOverride:double)
     Index=Get OverrideDefaultMobileFontSizeMultiplier() -> (OverrideDefaultMobileFontSizeMultiplier:bool)
    ) -> (ReturnValue:double)
   ) -> (then)
   n3 = Branch(execute=@n1.then_1, Condition=Get OverrideSecondaryTextSize() -> (OverrideSecondaryTextSize:bool))
   n4 = Break Slate Font Info!(SlateFontInfo:"Slate Font Info Structure (by ref)"=Get SecondaryFont() -> (SecondaryFont:Slate Font Info Structure)) -> (
    FontObject:Object Reference
    FontMaterial:Object Reference
    OutlineSettings:Font Outline Settings Structure
    TypefaceFontName:Name
    Size:float=~
    LetterSpacing:int
    SkewAmount:float=~
    bForceMonospaced:bool=~
    bMaterialIsStencil:bool=~
    MonospacedWidth:float=~
   ) {is=title K2Node_BreakStruct}
   n5 = Break Slate Font Info(SlateFontInfo=Get PrimaryFont()) -> (Size=~, SkewAmount=~, bForceMonospaced=~, bMaterialIsStencil=~, MonospacedWidth=~)
   n6 = Get OverrideSecondaryTextSizeMultiplier() -> (OverrideSecondaryTextSizeMultiplier:double)
   n7 = Get Block_SecondaryText()
   > SetFont(
    self=@n7
    Font=MakeSlateFontInfo!(
     FontObject:Object Reference=@n4
     FontMaterial:Object Reference=@n4.FontMaterial
     OutlineSettings:Font Outline Settings Structure=@n4.OutlineSettings
     TypefaceFontName:Name=@n4.TypefaceFontName
     Size:float=float * float(A=@n5.Size, B=@n6) {is=title K2Node_PromotableOperator}
     LetterSpacing:int=@n4.LetterSpacing
     SkewAmount:float=@n4.SkewAmount
     bForceMonospaced:bool=false
     bMaterialIsStencil:bool=false
     MonospacedWidth:float=1
    ) -> (ReturnValue:Slate Font Info Structure)
   )
   > SetMobileFontSizeMultiplier(
    self=@n7
    MobileFontSizeMultiplier=Select(
     Option 0=Get MobileFontSizeMultiplier()
     Option 1=Get MobileFontSizeMultiplierSecondaryOverride() -> (MobileFontSizeMultiplierSecondaryOverride:double)
     Index=Get OverrideDefaultMobileFontSizeMultiplier()
    )
   )
   n8 = Get Block_SecondaryText()
   SetFont(execute=@n3.else, self=@n8, Font=Get SecondaryFont())
   > SetMobileFontSizeMultiplier(
    self=@n8
    MobileFontSizeMultiplier=Select(
     Option 0:float=Get MobileFontSizeMultiplierSecondary() -> (MobileFontSizeMultiplierSecondary:double)
     Option 1:float=Get MobileFontSizeMultiplierSecondaryOverride()
     Index=Get OverrideDefaultMobileFontSizeMultiplier()
    ) -> (ReturnValue:float)
   )
  ]}
 UpdatePrompt {nodes [
   n1 = UpdatePrompt!() -> (then, InputType:ECommonInputType Enum) {is=title K2Node_FunctionEntry}
   > n2 = Sequence()
   > SetVisibility(self=[Get HoldProgressBar(), Get HoldIcon() -> (HoldIcon:Image Object Reference), Get Prompt()], InVisibility=Collapsed)
   > n3 = Set CurrentHoldProgress(CurrentHoldProgress=0)
   > SetScalarParameterValue(self=Get HoldProgressDynamicMaterial(), ParameterName=Percentage, Value=@n3)
   Set IsPromptHoldable(execute=@n2.then_1, IsPromptHoldable:bool=IsPromptHoldable() {is=title K2Node_CallFunction}) -> (Output_Get:bool)
   > Set IsPromptIcon(IsPromptIcon:bool=IsPromptIcon!() -> (Valid:bool) {is=title K2Node_CallFunction}) -> (Output_Get:bool)
   n4 = Get bHideInputActionWithKeyboard() -> (bHideInputActionWithKeyboard:bool)
   n5 = Get IsPromptHoldable() -> (IsPromptHoldable:bool)
   n6 = Get IsPromptIcon() -> (IsPromptIcon:bool)
   n7 = Select(
    Option 0:ESlateVisibility Enum=Collapsed
    Option 1:ESlateVisibility Enum=SelfHitTestInvisible
    Index=AND Boolean(A=OR Boolean(A=@n6, B=@n5), B=NOT Boolean(A=@n4))
   ) -> (ReturnValue:ESlateVisibility Enum)
   n8 = Select(Option 0=Collapsed, Option 1=SelfHitTestInvisible, Index=Get IsPromptHoldable())
   n9 = Select(Option 0=Collapsed, Option 1=SelfHitTestInvisible, Index=Get IsPromptIcon())
   > SetVisibility(
    self=Get Prompt()
    InVisibility=Select(
     Option 0=Collapsed
     Option 1=Select!(
      MouseAndKeyboard:ESlateVisibility Enum=@n7
      Gamepad:ESlateVisibility Enum=@n9
      Touch:ESlateVisibility Enum=@n8
      Count:ESlateVisibility Enum=Collapsed
      Index:ECommonInputType Enum=@n1
     ) -> (ReturnValue)
     Index=GetIsEnabled()
    )
   )
   > n10 = Branch(Condition=Get IsPromptHoldable())
   n11 = Get IsPromptIcon()
   > SetScalarParameterValue(
    self=Get HoldProgressDynamicMaterial()
    ParameterName=Shape Switch
    Value=Select(
     MouseAndKeyboard:float=Select!(Option 0:float=0, Option 1:float=1, Index:bool=@n11) -> (ReturnValue)
     Gamepad:float=0
     Touch:float=0
     Count:float=0
     Index=@n1
    ) -> (ReturnValue:float)
   )
   > SetVisibility(self=Get HoldProgressBar(), InVisibility=SelfHitTestInvisible)
   n12 = Get IsPromptIcon()
   > SetVisibility(
    self=Get HoldIcon()
    InVisibility=Select!(
     MouseAndKeyboard:ESlateVisibility Enum=Select!(Option 0:ESlateVisibility Enum=SelfHitTestInvisible, Option 1:ESlateVisibility Enum=Collapsed, Index:bool=@n12) -> (ReturnValue)
     Gamepad:ESlateVisibility Enum=Collapsed
     Touch:ESlateVisibility Enum=SelfHitTestInvisible
     Count:ESlateVisibility Enum=Collapsed
     Index:ECommonInputType Enum=@n1
    ) -> (ReturnValue:ESlateVisibility Enum)
   )
   SetScalarParameterValue(execute=@n10.else, self=Get HoldProgressDynamicMaterial(), ParameterName=Percentage, Value=0)
   > SetVisibility(self=Get HoldProgressBar(), InVisibility=Collapsed)
   > SetVisibility(self=Get HoldIcon(), InVisibility=Collapsed)
  ]}
 UpdatePromptVisibility {nodes [
   n1 = UpdatePromptVisibility!() -> (then, InputType:ECommonInputType Enum) {is=title K2Node_FunctionEntry}
   > Set IsPromptHoldable(IsPromptHoldable=IsPromptHoldable())
   > Set IsPromptIcon(IsPromptIcon=IsPromptIcon())
   n2 = Get bHideInputActionWithKeyboard()
   n3 = Get IsPromptHoldable()
   n4 = Get IsPromptIcon()
   n5 = Select(Option 0=Collapsed, Option 1=SelfHitTestInvisible, Index=AND Boolean(A=OR Boolean(A=@n4, B=@n3), B=NOT Boolean(A=@n2)))
   n6 = Select(Option 0=Collapsed, Option 1=SelfHitTestInvisible, Index=Get IsPromptHoldable())
   n7 = Select(Option 0=Collapsed, Option 1=SelfHitTestInvisible, Index=Get IsPromptIcon())
   > SetVisibility(
    self=Get Prompt()
    InVisibility=Select(
     Option 0=Collapsed
     Option 1=Select!(
      MouseAndKeyboard:ESlateVisibility Enum=@n5
      Gamepad:ESlateVisibility Enum=@n7
      Touch:ESlateVisibility Enum=@n6
      Count:ESlateVisibility Enum=Collapsed
      Index:ECommonInputType Enum=@n1
     ) -> (ReturnValue)
     Index=GetIsEnabled()
    )
   )
   > Return Node()
  ]}
]
variables [
 BackgroundMaterial {type=MaterialInstance, default=/FortUILibrary/UGC/MaterialInstances/Generic/MI_UEFN_UI_Bg_Loud}
 BorderMaterial {type=MaterialInstance, default=/FortUILibrary/UGC/MaterialInstances/Generic/MI_UEFN_UI_Outline}
 ButtonHeight {type=integer, default=56}
 Corner Radius {type=number, default=37}
 CurrentHoldProgress {type=number}
 CurrentInputType {type=ECommonInputType, default=MouseAndKeyboard}
 DoubleTextPadding {type=Margin, default=<ltrb 24 8 24 8>}
 DoubleTextSecondaryPadding {type=Margin, default=<ltrb 0 -6 0 0>}
 FadeInTime {type=number, default=0.1}
 FadeOutTime {type=number, default=0.1}
 GridUnit {type=integer, default=8}
 HoldCompletedSound {type=SoundBase, default=null}
 HoldPressedSound {type=SoundBase, default=null}
 HoldProgressDynamicMaterial {type=MaterialInstanceDynamic}
 HoldReleasedSound {type=SoundBase, default=null}
 HoldSound {type=SoundBase, default=null}
 HoldSoundAudioComponent {type=AudioComponent, default=null}
 HoveredSound {type=SoundBase, default=null}
 IsBorderAdditive {type=boolean}
 IsDarkMode {type=boolean, default=true}
 LastEnabledEvent {type=boolean}
 MobileButtonHeight {type=integer, default=54}
 MobileDoubleTextPadding {type=Margin, default=<ltrb 24 8 24 8>}
 MobileDoubleTextSecondaryPadding {type=Margin, default=<ltrb 0 -6 0 0>}
 MobileFontSizeMultiplier {type=number, default=1.3}
 MobileFontSizeMultiplierOverride {type=number, default=1}
 MobileFontSizeMultiplierSecondary {type=number, default=1.6}
 MobileFontSizeMultiplierSecondaryOverride {type=number, default=1}
 MobileMinHeight {type=integer}
 MobileMinWidth {type=integer}
 MobileSingleTextPadding {type=Margin, default=<ltrb 24 0 24 0>}
 OnFocusChanged {}
 OnHoldActionProgress {}
 OnHoldCompleted {}
 OverrideDefaultHeight {type=boolean}
 OverrideDefaultMobileFontSizeMultiplier {type=boolean}
 OverrideSecondaryTextSize {type=boolean}
 OverrideSecondaryTextSizeMultiplier {type=number, default=1}
 PressedSound {type=SoundBase, default=null}
 PressedSoundAudioComponent {type=AudioComponent, default=null}
 PrimaryFont {
  type=SlateFontInfo
  default {
   bForceMonospaced=false
   bMaterialIsStencil=false
   fontMaterial=/FortUILibrary/UGC/MaterialInstances/Generic/MI_UEFN_UI_TextColor_Loud
   fontObject=&HeadingNow
   letterSpacing=0
   monospacedWidth=1
   size=25.5
   skewAmount=0
   typefaceFontName=76 Bold
   outlineSettings {
    bApplyOutlineToDropShadows=false
    bMiteredCorners=false
    bSeparateFillAlpha=false
    outlineMaterial=null
    outlineSize=0
    outlineColor=<rgba 0 0 0 1>
   }
  }
 }
 PromptPadding {type=Margin, default=<ltrb 0 0 0 0>}
 Radius in Pixels {type=boolean, default=true}
 SecondaryFont {
  type=SlateFontInfo
  default {
   bForceMonospaced=false
   bMaterialIsStencil=false
   fontMaterial=&MI_UEFN_UI_TextColor_Loud
   fontObject=&HeadingNow
   letterSpacing=20
   monospacedWidth=1
   size=15.75
   skewAmount=0
   typefaceFontName=76 Bold
   outlineSettings {
    bApplyOutlineToDropShadows=false
    bMiteredCorners=false
    bSeparateFillAlpha=false
    outlineMaterial=null
    outlineSize=0
    outlineColor=<rgba 0 0 0 1>
   }
  }
 }
 SingleText {type=boolean, default=true}
 SingleTextPadding {type=Margin, default=<ltrb 24 0 24 0>}
]
widgets {
 info {inheritedWidgetCount=1, parentClass=/Script/FortniteUI.FortCTAButton, rootWidgetClass=&GridPanel, widgetCount=10}
 widgets [
  Root {widgetClassPath=&GridPanel, bIsVariable=true, widget=Root, properties {columnFill [1], rowFill [1]}}
  Block_Background {
   widgetClassPath=/FortUILibrary/UGC/Widgets/Buttons/Blocks/WBP_UEFN_Block_Background
   bIsVariable=true
   parent=Root
   widget=Block_Background
   slot {class=GridSlot}
  }
  Content {
   widgetClassPath=&GridPanel
   bIsVariable=true
   parent=Root
   widget=Content
   properties {columnFill [0,1,0], rowFill [1]}
   slot {class=GridSlot, properties {layer=10}}
  }
  Prompt {
   widgetClassPath=&GridPanel
   bIsVariable=true
   parent=Content
   widget=Prompt
   properties {columnFill [1], rowFill [1]}
   slot {class=GridSlot, properties {layer=20, padding=<ltrb 14 0 -24 0>}}
  }
  HoldProgressBar {
   widgetClassPath=&Image
   bIsVariable=true
   parent=Prompt
   widget=HoldProgressBar
   properties {
    visibility=Collapsed
    brush {
     bIsDynamicallyLoaded=false
     drawAs=Image
     imageType=FullColor
     mirroring=NoMirror
     resourceName=None
     resourceObject=/FortUILibrary/UGC/MaterialInstances/Generic/MI_UEFN_UI_InputActionProgress
     tiling=NoTile
     imageSize=<48 48>
     margin=<ltrb 0 0 0 0>
     outlineSettings {
      bUseBrushTransparency=false
      roundingType=HalfHeightRadius
      width=0
      color {colorUseRule=UseColor_Specified, specifiedColor=<rgba 0 0 0 0>}
      cornerRadii=<0 0 0 0>
     }
     tintColor {colorUseRule=UseColor_Specified, specifiedColor=<rgba 1 1 1 1>}
     uVRegion {bIsValid=false, max=<0 0>, min=<0 0>}
    }
   }
   slot {class=GridSlot, properties {horizontalAlignment=HAlign_Center, layer=10, verticalAlignment=VAlign_Center}}
  }
  HoldIcon {
   widgetClassPath=&Image
   bIsVariable=true
   parent=Prompt
   widget=HoldIcon
   properties {
    visibility=Collapsed
    brush {
     bIsDynamicallyLoaded=false
     drawAs=Image
     imageType=FullColor
     mirroring=NoMirror
     resourceName=None
     resourceObject=/FortUILibrary/UGC/MaterialInstances/Generic/MI_UEFN_UI_MousePressAndHold
     tiling=NoTile
     imageSize=<28 28>
     margin=<ltrb 0 0 0 0>
     outlineSettings {
      bUseBrushTransparency=false
      roundingType=HalfHeightRadius
      width=0
      color {colorUseRule=UseColor_Specified, specifiedColor=<rgba 0 0 0 0>}
      cornerRadii=<0 0 0 0>
     }
     tintColor {colorUseRule=UseColor_Specified, specifiedColor=<rgba 1 1 1 1>}
     uVRegion {bIsValid=false, max=<0 0>, min=<0 0>}
    }
   }
   slot {class=GridSlot, properties {horizontalAlignment=HAlign_Center, layer=30, verticalAlignment=VAlign_Center}}
  }
  InputActionWidget {
   widgetClassPath=&CommonActionWidget
   bInherited=true
   bIsVariable=true
   parent=Prompt
   widget=InputActionWidget
   properties {
    progressMaterialParam=Percentage
    visibility=HitTestInvisible
    progressMaterialBrush {
     bIsDynamicallyLoaded=false
     drawAs=Image
     imageType=NoImage
     mirroring=NoMirror
     resourceName=None
     resourceObject=/Game/UI/Materials/EmptyProgressMaterial
     tiling=NoTile
     imageSize=<48 48>
     margin=<ltrb 0 0 0 0>
     outlineSettings {
      bUseBrushTransparency=false
      roundingType=HalfHeightRadius
      width=0
      color {colorUseRule=UseColor_Specified, specifiedColor=<rgba 0 0 0 0>}
      cornerRadii=<0 0 0 0>
     }
     tintColor {colorUseRule=UseColor_Specified, specifiedColor=<rgba 1 1 1 1>}
     uVRegion {bIsValid=false, max=<0 0>, min=<0 0>}
    }
   }
   slot {class=GridSlot, properties {horizontalAlignment=HAlign_Center, layer=40, verticalAlignment=VAlign_Center}}
  }
  VerticalBox_Text {
   widgetClassPath=&VerticalBox
   bIsVariable=true
   parent=Content
   widget=VerticalBox_Text
   slot {class=GridSlot, properties {column=1, verticalAlignment=VAlign_Center, padding=<ltrb 8 0 8 0>}}
  }
  Block_Text {
   widgetClassPath=/FortUILibrary/UGC/Widgets/Buttons/Blocks/WBP_UEFN_Block_Text
   bIsVariable=true
   parent=VerticalBox_Text
   widget=Block_Text
   slot {class=VerticalBoxSlot, properties {verticalAlignment=VAlign_Center}}
  }
  Block_SecondaryText {
   widgetClassPath=&WBP_UEFN_Block_Text
   bIsVariable=true
   parent=VerticalBox_Text
   widget=Block_SecondaryText
   slot {class=VerticalBoxSlot, properties {horizontalAlignment=HAlign_Center, verticalAlignment=VAlign_Top}}
  }
 ]
}
